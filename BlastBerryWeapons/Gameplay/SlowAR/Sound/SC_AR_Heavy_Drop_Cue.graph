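# Sound cue for dropping the SlowAR heavy assault rifle, routed to the loot collect sound class with weapon drop attenuation and concurrency.
# A mixer layers one of three random drop waves with one of three random sweetener waves, the sweetener delayed by a short fixed time.
# Each layer passes through a modulator, and the cue plays slightly raised in pitch and reduced in volume.
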SoundCue /BlastBerryWeapons/Gameplay/SlowAR/Sound/SC_AR_Heavy_Drop_Cue
root=Mixer_0
nodes [
 {id=Delay_0, children [Modulator_0], properties {delayMax=0.1, delayMin=0.1}}
 {id=Mixer_0, children [Modulator_1,Delay_0], properties {inputVolume [1,1]}}
 {id=Modulator_0, children [Random_1]}
 {id=Modulator_1, children [Random_0]}
 {id=Random_0, children [WavePlayer_0,WavePlayer_4,WavePlayer_5], properties {weights [1,1,1]}}
 {id=Random_1, children [WavePlayer_1,WavePlayer_2,WavePlayer_3], properties {weights [1,1,1]}}
 {id=WavePlayer_0, properties {soundWaveAssetPtr=/BlastBerryWeapons/Gameplay/SlowAR/Sound/Wavs/Paprika_AR_Heavy_Drop_01}}
 {id=WavePlayer_1, properties {soundWaveAssetPtr=/BlastBerryWeapons/Gameplay/SlowAR/Sound/Wavs/Paprika_AR_Heavy_Drop_Sweetner_01}}
 {id=WavePlayer_2, properties {soundWaveAssetPtr=/BlastBerryWeapons/Gameplay/SlowAR/Sound/Wavs/Paprika_AR_Heavy_Drop_Sweetner_02}}
 {id=WavePlayer_3, properties {soundWaveAssetPtr=/BlastBerryWeapons/Gameplay/SlowAR/Sound/Wavs/Paprika_AR_Heavy_Drop_Sweetner_03}}
 {id=WavePlayer_4, properties {soundWaveAssetPtr=/BlastBerryWeapons/Gameplay/SlowAR/Sound/Wavs/Paprika_AR_Heavy_Drop_02}}
 {id=WavePlayer_5, properties {soundWaveAssetPtr=/BlastBerryWeapons/Gameplay/SlowAR/Sound/Wavs/Paprika_AR_Heavy_Drop_03}}
]
properties {
 attenuationSettings=&Fort_Weapon_Drop_Att
 maxDistance=3000
 pitchMultiplier=1.1
 soundClassObject=&Fort_Loot_Collect
 volumeMultiplier=0.6
 concurrencyOverrides {resolutionRule=PreventNew}
 concurrencySet [&Weapon_Drop_StopOldest]
}
