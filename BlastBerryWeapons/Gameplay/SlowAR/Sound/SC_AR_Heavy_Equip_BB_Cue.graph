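# Sound cue for the equip foley of the heavy, slow assault rifle in BlastBerryWeapons.
# A Random node picks one of six equally weighted heavy AR equip wave variants and feeds a root Modulator.
# It plays at reduced volume in the weapon foley sound class, borrows the shared weapon reload attenuation and overrides concurrency.
# The debug flag is left enabled on the cue.
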SoundCue /BlastBerryWeapons/Gameplay/SlowAR/Sound/SC_AR_Heavy_Equip_BB_Cue
root=Modulator_0
nodes [
 {id=Modulator_0, children [Random_0]}
 {
  id=Random_0
  children [WavePlayer_6, WavePlayer_7, WavePlayer_8, WavePlayer_9, WavePlayer_10, WavePlayer_11]
  properties {weights [1,1,1,1,1,1]}
 }
 {id=WavePlayer_10, properties {soundWaveAssetPtr=/BlastBerryWeapons/Gameplay/SlowAR/Sound/Wavs/Paprika_AR_Heavy_Equip_05}}
 {id=WavePlayer_11, properties {soundWaveAssetPtr=/BlastBerryWeapons/Gameplay/SlowAR/Sound/Wavs/Paprika_AR_Heavy_Equip_06}}
 {id=WavePlayer_6, properties {soundWaveAssetPtr=/BlastBerryWeapons/Gameplay/SlowAR/Sound/Wavs/Paprika_AR_Heavy_Equip_01}}
 {id=WavePlayer_7, properties {soundWaveAssetPtr=/BlastBerryWeapons/Gameplay/SlowAR/Sound/Wavs/Paprika_AR_Heavy_Equip_02}}
 {id=WavePlayer_8, properties {soundWaveAssetPtr=/BlastBerryWeapons/Gameplay/SlowAR/Sound/Wavs/Paprika_AR_Heavy_Equip_03}}
 {id=WavePlayer_9, properties {soundWaveAssetPtr=/BlastBerryWeapons/Gameplay/SlowAR/Sound/Wavs/Paprika_AR_Heavy_Equip_04}}
]
properties {
 attenuationSettings=&Fort_Weapon_Reload_Att
 bDebug=true
 bOverrideConcurrency=true
 maxDistance=2250
 soundClassObject=&Fort_WeaponFoley
 volumeMultiplier=0.301
}
